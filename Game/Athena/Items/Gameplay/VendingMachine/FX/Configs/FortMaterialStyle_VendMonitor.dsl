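; Material style config for vending machines, named for the monitor, that pushes a set of style properties onto a target vending machine prop.
; ApplyTo takes the target from the apply params and passes it, with this asset's Properties struct, to ApplyMaterialStyleToVendingMachine.
; That helper casts the target to BP_StationProp_VendingMachine_Emote_Base and calls TriggerChangeMaterialStyle, leaving the visual change to the prop itself.
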(fn ApplyMaterialStyleToVendingMachine
  (params (Target :type Object%20Reference) (DesiredStyleProperties :type Vending%20Machine%20Material%20Style%20Structure))
  (bind _asbp_station_prop_vending_machine_emote_base (Utilities|Casting|CastToBP_StationProp_VendingMachine_Emote_Base Target))
  (Class|BPStationPropVendingMachineEmoteBase|TriggerChangeMaterialStyle _asbp_station_prop_vending_machine_emote_base DesiredStyleProperties))

(fn ApplyTo
  (params (Params :type Fort%20Material%20Style%20Apply%20Params%20Structure%20%28by%20ref%29))
  (CallFunction|ApplyMaterialStyletoVendingMachine (Utilities|Struct|BreakFortMaterialStyleApplyParams Params) (Variables|Default|GetProperties)))
